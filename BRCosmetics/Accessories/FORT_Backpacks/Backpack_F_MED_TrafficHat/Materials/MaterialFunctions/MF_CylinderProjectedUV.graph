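# Material function in the traffic hat backpack's folder that builds a tiled, cylinder-style projected UV from pre-skinning local position.
# The offset position is vertex-interpolated and scaled down; arctangent over pi gives U, though both arguments read G, and B gives V.
# A Vector2 input sets tiling, with U floored to whole repeats.
# It also outputs mesh-UV derivatives; radius-normalized height and world-position branches are unconnected.
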MaterialFunction /BRCosmetics/Accessories/FORT_Backpacks/Backpack_F_MED_TrafficHat/Materials/MaterialFunctions/MF_CylinderProjectedUV
graph {nodes [
  n1 = Add(A=LocalPosition() {localOrigin=InstancePreSkinning}.XYZ, B=Constant3Vector() {constant=<rgba 0 6 4 0>})
  n2 = MaterialFunctionCall(Color=Divide(A=VertexInterpolator(VS=@n1).PS) {constB=100}) {materialFunction=&SplitComponents}
  n3 = ObjectRadius()
  Add(A=Divide(A=@n2.B, B=Multiply(A=@n3) {constB=2})) {constB=0.5}
  n4 = Divide(A=Arctangent2Fast(Y=@n2.G, X=@n2.G), B=MaterialFunctionCall() {materialFunction=&Pi}.Result)
  n5 = AppendVector(A=@n4, B=@n2.B)
  n6 = FunctionInput() {bUsePreviewValueAsDefault=true, inputType=FunctionInput_Vector2, previewValue=<1 1 0 1>}
  n7 = Floor(None=ComponentMask(None=@n6) {r=true})
  n8 = ComponentMask(None=@n6) {g=true}
  FunctionOutput(None=Multiply(A=@n5, B=AppendVector(A=@n7, B=@n8)))
  n9 = TextureCoordinate()
  FunctionOutput(None=DDX(Value=@n9)) {outputName=DDX, sortPriority=1}
  FunctionOutput(None=DDY(Value=@n9)) {outputName=DDY, sortPriority=2}
  LocalPosition()
  Multiply(A=@n4) {constB=2}
  OneMinus(None=@n5)
  n10 = WorldPosition() {worldPositionShaderOffset=WPT_CameraRelative}
  n11 = ObjectPositionWS() {originType=CameraRelative}
  Transform(None=Divide(A=Subtract(A=@n10.XYZ, B=@n11)) {constB=100}) {transformSourceType=TRANSFORMSOURCE_World, transformType=TRANSFORM_Local}
 ]}
